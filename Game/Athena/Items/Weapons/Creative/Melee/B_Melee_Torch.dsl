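; Creative torch melee weapon that lights its surroundings and spreads Curie fire to nearby actors while equipped.
; Equipping starts fire audio and picks a point light or mobile fake light; particles follow weapon visibility.
; A timer sets loop audio "Speed" from the greater of player 0's smoothed velocity and the torch's turning.
; An authority-only timer async sphere-overlaps at a stance-based offset, applying fire to hits; unequip stops both timers.
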(event EventOnWeaponVisibilityChanged
  (params (bVisible :type Boolean) (bSetForLocalControllerOnly :type Boolean))
  (bind _particlesystem (Variables|Default|GetParticleSystem))
  (|CallParentFunction bVisible false)
  (if bVisible
    (Components|Activation|Activate _particlesystem false)
    (else
      (Components|Activation|Deactivate _particlesystem))))

(event FortWeapon|EventOnLocalUnEquip
  ;[0.1056 1 0.6084] On Local UnEquip
  (Utilities|Time|ClearandInvalidateTimerbyHandle (Variables|Default|GetHandle%20Torch%20Loop%20Audio))
  (Audio|PlaySoundatLocation "/Game/Sounds/Creative/Gadgets/Torch/Torch_Fire_Stop_Cue.Torch_Fire_Stop_Cue" (Transformation|GetActorLocation))
  (bind _torch_fire_loop_cue (Variables|Default|GetTorch_Fire_Loop_Cue)
    (:then
      (Audio|Components|Audio|FadeOut _torch_fire_loop_cue 0.5 0.0 "Linear" :!class /Script/Engine.AudioComponent)
      (Networking|SwitchHasAuthority
        (:Authority
          (Utilities|Time|ClearandInvalidateTimerbyHandle (Variables|Curie|GetHandle%20Fire%20Loop%20Curie)))))
    (:else)))

(event FortWeapon|EventOnEquippedAndAttached
  ;[0.04473 0.7065 1] On Equipped and Attached
  (bind _returnvalue (Utilities|Time|SetTimerbyEvent (EventDispatchers|CreateEvent :!event Custom|Play%20Torch%20Audio%20Loop) 0.1 true))
  (Variables|Default|SetHandle%20Torch%20Loop%20Audio _returnvalue)
  (Audio|Components|Audio|FadeIn (Variables|Default|GetTorch_Fire_Loop_Cue) 0.5 1.0 0.0 "Linear" :!class /Script/Engine.AudioComponent)
  (Audio|Components|Audio|Play (Variables|Default|GetTorch_Fire_Start_Cue) 0.0)
  (CallFunction|SetLightComponents)
  (Networking|SwitchHasAuthority
    (:Authority
      (bind _returnvalue_2 (Utilities|Time|SetTimerbyEvent (EventDispatchers|CreateEvent :!event Custom|Update%20Torch%20Fire%20Loop) (Variables|Curie|GetFireLoopTime) true))
      (Variables|Curie|SetHandle%20Fire%20Loop%20Curie _returnvalue_2))))

(event Custom|Play%20Torch%20Audio%20Loop
  (bind _returnvalue (Transformation|GetActorForwardVector))
  (bind _velocity_smoothed (Variables|Default|GetVelocity%20Smoothed))
  (bind _returnvalue_2 (Math|Float|MapRangeClamped _velocity_smoothed 0.4 6e+02 0.4 1.3))
  (bind _rotation_smoothed (Variables|Default|GetRotation%20Smoothed))
  (bind _returnvalue_3 (Math|Float|MapRangeClamped _rotation_smoothed 0.0 1.5 0.0 1.5))
  (bind _returnvalue_4 (Game|GetPlayerPawn 0))
  (bind (_asfort_player_pawn _bsuccess) (Utilities|Casting|CastToFortPlayerPawn _returnvalue_4))
  (bind _torch_fire_loop_cue (Variables|Default|GetTorch_Fire_Loop_Cue))
  ;[1 1 1] Update Torch audio based on player movement
  (if _bsuccess
    (Variables|Default|SetVelocity%20Smoothed (Math|Interpolation|FInterpTo (Variables|Default|GetVelocity%20Smoothed) (Math|Vector|VectorLength (Transformation|GetVelocity _asfort_player_pawn)) (Utilities|Time|GetWorldDeltaSeconds) 5.0))
    (Variables|Default|SetRotation%20Smoothed (Math|Interpolation|FInterpTo (- 1.0 (Math|Float|Absolute(Float) (Math|Vector|DotProduct _returnvalue (Math|Vector|MakeVector (Variables|Default|GetCurrent%20Vector%20X) (Variables|Default|GetCurrent%20Vector%20Y) (Variables|Default|GetCurrent%20Vector%20Z))))) 1.5 (Utilities|Time|GetWorldDeltaSeconds) 1e+01))
    (Variables|Default|SetCurrent%20Vector%20X (.x _returnvalue))
    (Variables|Default|SetCurrent%20Vector%20Y (.y _returnvalue))
    (Variables|Default|SetCurrent%20Vector%20Z (.z _returnvalue))
    (if (>= _returnvalue_2 _returnvalue_3)
      (Audio|Parameter|SetFloatParameter _torch_fire_loop_cue "Speed" _returnvalue_2 :!class /Script/AudioExtensions.AudioParameterControllerInterface)
      (else
        (Audio|Parameter|SetFloatParameter _torch_fire_loop_cue "Speed" _returnvalue_3 :!class /Script/AudioExtensions.AudioParameterControllerInterface)))))

(event Custom|Update%20Torch%20Fire%20Loop
  ;[0.1845 1 0.04048] Update Torch Fire based on location
  (Networking|SwitchHasAuthority
    (:Authority
      (if (Curie|IsCurieActiveforActor self)
        (bind _firecenter (CallFunction|GetFireCenter))
        (Class|AsyncOverlapQueue|AsyncSphereOverlapActors _firecenter (* (Variables|Curie|GetTraceRadius) (Math|Vector|VectorLength (Transformation|GetActorScale3D))) (Utilities|Array|MakeArray self (Actor|GetOwner)) (EventDispatchers|CreateEvent :!event On%20Damage%20Canceled) (EventDispatchers|CreateEvent :!event DoDamageEffectTotarget) (EventDispatchers|CreateEvent :!event Custom|OnDamageCompleted))))))

(event Custom|DoDamageEffectToTarget
  (params (HitActor :type Actor%20Object%20Reference))
  ;[1 0.8386 0.4106] Do Damage Effect To Target
  (Utilities|IsValid HitActor
    (:Is%20Valid
      (Curie|ExecuteFireApplication (Actor|GetOwner) self HitActor (Variables|Curie|GetMagnitude) (Collision|MakeHitResult false true 0.0 0.0 (Transformation|GetActorLocation) (Transformation|GetActorLocation) "0,0,1" "0,0,1" 0 HitActor) (Variables|Curie|GetPropagationFuel) (Variables|Curie|GetLandscapeRadius)))))

(event Custom|OnDamageCompleted
  (params (WasCancelled :type Boolean)))

(fn GetFireCenter
  (bind _returnvalue (Transformation|GetActorLocation))
  ;[1 0.4584 0.06801] Get Fire Center in world coordinates
  (bind _aspawn (Utilities|Casting|CastToPawn (Actor|GetOwner))
    (:then
      (if (AI|Components|NavMovement|IsCrouching (Pawn|GetMovementComponent _aspawn))
        (return (+ (Variables|Curie|GetFireOffsetCrouching) _returnvalue))
        (else
          _)))
    (:CastFailed
      _))
  (return (+ _returnvalue (Variables|Curie|GetFireOffsetStanding))))

(fn OnDamageCanceled
  ;[1 0.8401 0.3886] On Damage Canceled
  (return false))

(fn SetLightComponents
  (Networking|CanExecuteCosmeticEvents
    (:True
      (if (Fortnite|IsMobilePlatform)
        (Rendering|SetVisibility (Variables|Default|GetMobileFakeLight) true false)
        (Rendering|SetVisibility (Variables|Default|GetPointLight) false false)
        (else
          (Rendering|SetVisibility (Variables|Default|GetPointLight) true false)
          (Rendering|SetVisibility (Variables|Default|GetMobileFakeLight) false false))))))

(fn ConstructionScript
  (|CallParentFunction))
