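# Deferred decal material for a creative melee torch, named as a mobile variant, that appears to fake a warm light glow on nearby surfaces.
# It is unlit and translucent, with a sphere mask between actor position and world position setting the falloff over a LightRadius parameter.
# The mask is multiplied by itself to steepen the falloff, then tinted by the orange LightColor parameter, avoiding any light calculation.
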Material /Game/Athena/Items/Weapons/Creative/Melee/MD_TorchlightMobile
graph {
 properties {
  blendMode=BLEND_Translucent
  materialDomain=MD_DeferredDecal
  pixelDepthOffsetMode=PDOM_Legacy
  shadingModel=MSM_Unlit
  usedShadingModels=Unlit
 }
 nodes [
  n1 = SphereMask(A=ActorPositionWS(), B=WorldPosition().XYZ, Radius=ScalarParameter() {defaultValue=300, parameterName=LightRadius}, Hardness=Constant())
  Multiply(A=VectorParameter() {parameterName=LightColor, defaultValue=<rgba 0.3 0.0983 0.0079 1>}.RGB, B=Multiply(A=@n1, B=@n1))
 ]
}
